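# Screen material for the FNCS Shield S24 back bling, showing the FNCS logo or, once Char0 is above zero, an animated emissive numeric counter.
# Char0 and Char1 offset a number-atlas lookup per sphere-masked column, over glitch-distorted, rotated UVs.
# A Timestamp parameter feeds an interpolated transition time, with reused Jumbotron curve-atlas rows animating celebration bursts, streaks, ribbons, sweep, flashes and radial rings.
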Material /BRCosmetics/Accessories/FORT_Backpacks/Backpack_M_MED_FNCSShield_S24/Materials/M_MED_FNCSShield_S24_Screen_A
graph {
 properties {
  bUseMaterialAttributes=true
  bUsedWithMeshDeformer=true
  bUsedWithSkeletalMesh=true
  pixelDepthOffsetMode=PDOM_Legacy
  usedShadingModels=Default Lit
 }
 nodes [
  n1 = TextureSample(UVs=NamedRerouteUsage()) {
   constCoordinate=1
   samplerType=SAMPLERTYPE_Masks
   texture=/BRCosmetics/Accessories/FORT_Backpacks/Backpack_M_MED_FNCSShield_S24/Textures/T_FNCSShield_Logo_Msk
  }
  Clamp(None=OneMinus(None=@n1.R))
  Constant3Vector() {constant=<rgba 0 0 1 0>}
  Frac(None=Divide(A=Time(), B=ScalarParameter() {defaultValue=3, parameterName=TransitionLength}))
  n2 = Time()
  n3 = Saturate(None=Sign(None=Sine(None=@n2) {period=0.25}))
  n4 = CurveAtlasRowParameter(CurveTime=NamedRerouteUsage()) {
   atlas=/BRCosmetics/Accessories/FORT_Backpacks/Backpack_M_MED_Jumbotron_S20/Textures/CA_JumboTron
   curve=/Game/Accessories/FORT_Backpacks/Backpack_M_MED_Jumbotron_S20/Textures/C_JumboTron_Flashes
   defaultValue=4
   parameterName=Flashes
   sortPriority=0
  }
  n5 = NamedRerouteUsage()
  n6 = Constant() {r=0.5}
  n7 = ScalarParameter() {defaultValue=1.25, parameterName=BG_SphereMsk_Size}
  n8 = ScalarParameter() {defaultValue=0.3, parameterName=BG_SphereMsk_Falloff}
  n9 = LinearInterpolate(
   A=VectorParameter() {parameterName=BGColorOuter, defaultValue=<rgba 0 0.8632 0.107 1>}.RGB
   B=VectorParameter() {parameterName=BGColorInner, defaultValue=<rgba 0 0 0 1>}.RGB
   Alpha=Power(Base=SphereMask(A=@n5, B=@n6, Radius=@n7, Hardness=@n8), Exp=Constant() {r=1.5})
  )
  n10 = NamedRerouteUsage()
  n11 = LinearInterpolate(
   A=LinearInterpolate(A=Add(A=@n9, B=@n10), B=VectorParameter() {parameterName=CelebrationBurstColor, defaultValue=<rgba 2 2 2 0>}.RGB, Alpha=NamedRerouteUsage())
   B=VectorParameter() {group=Radial Effect, parameterName=RadialEffectColor, defaultValue=<rgba 0.0186 0.3073 0 1>}.RGB
   Alpha=NamedRerouteUsage()
  )
  n12 = LinearInterpolate(
   A=@n11
   B=LinearInterpolate(
    A=VectorParameter() {parameterName=TextColor, defaultValue=<rgba 0.0204 0.8698 0.0727 1>}.RGB
    B=VectorParameter() {parameterName=TextColorCelebrate, defaultValue=<rgba 9.3165 30 7.7773 1>}.RGB
    Alpha=Multiply(A=@n3, B=@n4.R)
   )
   Alpha=NamedRerouteUsage()
  )
  n13 = VectorParameter() {parameterName=DropShadowColor, defaultValue=<rgba 0.0002 0.0104 0.0009 1>}
  n14 = NamedRerouteUsage()
  n15 = LinearInterpolate(
   A=LinearInterpolate(A=@n11, B=LinearInterpolate(A=@n12, B=@n13.RGB, Alpha=@n14), Alpha=NamedRerouteUsage())
   B=VectorParameter() {parameterName=SweepColor, defaultValue=<rgba 1.3736 5 1.2714 1>}.RGB
   Alpha=NamedRerouteUsage()
  )
  n16 = ScalarParameter() {defaultValue=0.9, parameterName=CRT_Amount}
  LinearInterpolate(A=@n15, B=Multiply(A=@n15, B=NamedRerouteUsage()), Alpha=@n16)
  LinearInterpolate(A=@n11, B=Multiply(A=@n11, B=NamedRerouteUsage()), Alpha=@n16)
  LinearInterpolate(B=Multiply(B=@n10))
  LinearInterpolate(
   A=@n11
   B=VectorParameter() {parameterName=CelebrationStreaksColor, defaultValue=<rgba 0.1085 0.3968 0.0118 1>}.RGB
   Alpha=NamedRerouteUsage()
  )
  MaterialFunctionCall!(
   Time
   Divisor=ScalarParameter() {defaultValue=1.5, parameterName=DebugTimeDivisor}
   Mod=ScalarParameter() {defaultValue=2, parameterName=DebugTimeModulo}
  ) {materialFunction=/Game/Effects/Fort_Effects/Materials/_Functions/MF_ModuloTime}
  n17 = VectorParameter() {group=Jumbotron, parameterName=FNCS Logo Shadow Color, defaultValue=<rgba 0.0052 0.0052 0.0052 0.9>}
  n18 = VectorParameter() {group=Jumbotron, parameterName=FNCS Logo Color, defaultValue=<rgba 0.0204 0.8698 0.0727 1>}
  n19 = OneMinus(None=@n1.R)
  n20 = OneMinus(None=Multiply(A=@n17.A, B=@n1.B))
  n21 = LinearInterpolate(A=LinearInterpolate(A=@n17.RGB, B=@n18.RGB, Alpha=@n1.R), B=NamedRerouteUsage(), Alpha=Saturate(None=Multiply(A=@n19, B=@n20)))
  n22 = Constant() {r=1}
  n23 = If(A=ScalarParameter() {parameterName=Char0}, "A > B"=Constant(), "A == B"=@n22, "A < B"=@n22)
  n24 = SetMaterialAttributes!(
   MaterialAttributes
   Base Color=LinearInterpolate(A=Constant3Vector(), B=@n21, Alpha=@n23)
   Emissive Color=LinearInterpolate(A=NamedRerouteUsage(), B=@n21, Alpha=@n23)
   Metallic=NamedRerouteUsage()
   Specular=NamedRerouteUsage()
   Roughness=NamedRerouteUsage()
   Normal=NamedRerouteUsage()
  ) {attributeSetTypes [
    #1
    #2
    #3
    #4
    #5
    #6
   ]}
  MaterialFunctionCall(In=@n24) {materialFunction=&MF_TechArtOverrides_accessories}
  Multiply(A=@n18.RGB, B=NamedRerouteDeclaration(None=@n1.R) {name=screenMask, nodeColor=<rgba 1 0 0.5391 1>})
  n25 = ScalarParameter() {defaultValue=50, parameterName=CRTScaleUV}
  Multiply(A=@n25) {constB=0.5}
  n26 = MaterialFunctionCall(Float2=NamedRerouteUsage()) {materialFunction=&BreakOutFloat2Components}
  n27 = Add(A=ScalarParameter() {defaultValue=-0.578, parameterName=U_Mask_Bias}, B=@n26.R)
  n28 = Divide(A=NamedRerouteUsage(), B=ScalarParameter() {defaultValue=2.8, parameterName=U_Mask_Divisor})
  n29 = Constant() {r=1}
  n30 = Divide(A=NamedRerouteUsage())
  n31 = Add(A=@n30, B=Divide(A=@n30) {constB=-20, desc=?}) {constB=-0.01}
  n32 = Divide(A=@n31)
  NamedRerouteDeclaration(None=Min(A=SphereMask(A=@n27, B=@n28, Radius=@n28, Hardness=@n29), B=SphereMask(A=@n26.G, B=@n31, Radius=@n32, Hardness=@n29))) {name=NumbersBoxMask, nodeColor=<rgba 0 1 0.8828 1>}
  n33 = MaterialFunctionCall(Float2=NamedRerouteUsage()) {materialFunction=&BreakOutFloat2Components}
  n34 = CurveAtlasRowParameter(CurveTime=NamedRerouteUsage()) {
   atlas=&CA_JumboTron
   curve=/Game/Accessories/FORT_Backpacks/Backpack_M_MED_Jumbotron_S20/Textures/C_JumboTron_Sweep
   defaultValue=5
   parameterName=Sweep
   sortPriority=0
  }
  n35 = Add(A=@n34.R) {constB=-1, desc=Curves can't go negative . . .}
  n36 = Constant() {r=0.9}
  n37 = Constant()
  NamedRerouteDeclaration(None=Power(Base=SphereMask(A=@n33.R, B=@n35, Radius=@n36, Hardness=@n37)) {constExponent=2.5}) {name=Sweep, nodeColor=<rgba 0.3594 0 1 1>}
  n38 = NamedRerouteUsage()
  n39 = ScalarParameter() {defaultValue=16, parameterName=Pixelation}
  n40 = Divide(A=Ceil(None=Multiply(A=@n38, B=@n39)), B=@n39)
  n41 = Time()
  n42 = Constant2Vector() {g=0.12, r=0.34}
  n43 = TextureSample(UVs=Add(A=TextureCoordinate(), B=Multiply(A=@n41, B=@n42))) {samplerType=SAMPLERTYPE_LinearColor, texture=/Game/Accessories/FORT_Backpacks/Backpack_M_MED_Jumbotron_S20/Textures/T_Jumbotron_CRT}
  n44 = Time()
  n45 = Constant2Vector() {g=0.26, r=0.12}
  n46 = TextureSample(UVs=Add(A=TextureCoordinate(), B=Multiply(A=@n44, B=@n45))) {samplerType=SAMPLERTYPE_LinearColor, texture=&T_Jumbotron_CRT}
  n47 = Multiply(
   A=Add(A=AppendVector(A=@n43.R, B=@n43.R), B=AppendVector(A=@n46.R, B=@n46.R))
   B=ScalarParameter() {defaultValue=0.04, group=Radial Effect, parameterName=Distortion, sliderMax=0.09}
  ) {constB=0}
  n48 = CurveAtlasRowParameter(CurveTime=NamedRerouteUsage()) {
   atlas=&CA_JumboTron
   curve=/Game/Accessories/FORT_Backpacks/Backpack_M_MED_Jumbotron_S20/Textures/C_Jumbotron_RadialEffect_S21
   defaultValue=6
   parameterName=RadialEffect
   sortPriority=0
  }
  n49 = ScalarParameter() {defaultValue=8, group=Radial Effect, parameterName=Pulse Rate}
  NamedRerouteDeclaration(None=MaterialFunctionCall(UVs=Subtract(A=@n40, B=@n47), Texture Scale=Multiply(A=@n48, B=@n49)) {materialFunction=&ScaleUVsByCenter}.UVs) {name=RadialUV, nodeColor=<rgba 1 0 0.3984 1>}
  n50 = TextureSampleParameter2D() {
   group=Jumbotron
   parameterName=Metal Texture
   samplerType=SAMPLERTYPE_Masks
   texture=/BRCosmetics/Accessories/FORT_Backpacks/Backpack_M_MED_FNCSShield_S24/Textures/T_M_MED__FNCSShield_S24_Backpack_M
  }
  NamedRerouteDeclaration(None=@n50.R) {name=outSpecular, nodeColor=<rgba 1 0.1641 0 1>}
  n51 = NamedRerouteUsage()
  n52 = ScalarParameter() {defaultValue=0.2, group=Radial Effect, parameterName=Outer Radius A}
  n53 = Constant() {r=100}
  n54 = ScalarParameter() {defaultValue=0.18, group=Radial Effect, parameterName=Inner Radius A}
  n55 = NamedRerouteUsage()
  n56 = ScalarParameter() {defaultValue=0.14, group=Radial Effect, parameterName=Outer Radius B}
  n57 = Constant() {r=100}
  n58 = ScalarParameter() {defaultValue=0.13, group=Radial Effect, parameterName=Inner Radius B}
  n59 = Add(
   A=Subtract(
    A=MaterialFunctionCall(UVs=@n51, Radius=@n52, Density=@n53) {desc=Outer Circle, materialFunction=&RadialGradientExponential}.RadialGradientExponential
    B=MaterialFunctionCall(UVs=@n51, Radius=@n54, Density=@n53) {desc=Inner Circle, materialFunction=&RadialGradientExponential}.RadialGradientExponential
   )
   B=Subtract(
    A=MaterialFunctionCall(UVs=@n55, Radius=@n56, Density=@n57) {desc=Outer Circle, materialFunction=&RadialGradientExponential}.RadialGradientExponential
    B=MaterialFunctionCall(UVs=@n55, Radius=@n58, Density=@n57) {desc=Inner Circle, materialFunction=&RadialGradientExponential}.RadialGradientExponential
   )
  )
  n60 = NamedRerouteUsage()
  n61 = Constant() {r=100}
  n62 = Subtract(
   A=MaterialFunctionCall(UVs=@n60, Radius=ScalarParameter() {defaultValue=0.1, group=Radial Effect, parameterName=Outer Radius C}, Density=@n61) {desc=Outer Circle, materialFunction=&RadialGradientExponential}.RadialGradientExponential
   B=MaterialFunctionCall(UVs=@n60, Radius=ScalarParameter() {defaultValue=0.09, group=Radial Effect, parameterName=Inner Radius C}, Density=@n61) {desc=Inner Circle, materialFunction=&RadialGradientExponential}.RadialGradientExponential
  )
  NamedRerouteDeclaration(None=Multiply(A=Add(A=@n59, B=@n62), B=ScalarParameter() {defaultValue=0.9, group=Radial Effect, parameterName=Radial Effect Opacity, sliderMax=1}) {constB=0.05}) {name=RadialEffect, nodeColor=<rgba 1 0 0.1935 1>}
  NamedRerouteDeclaration(None=@n50.G) {name=outMetal, nodeColor=<rgba 1 0.4453 0 1>}
  NamedRerouteDeclaration(None=@n50.B) {name=outRoughness, nodeColor=<rgba 1 0.9141 0 1>}
  n63 = LinearInterpolate(
   A=ScalarParameter() {defaultValue=0.55, parameterName=FinalFresnelLow}
   Alpha=MaterialFunctionCall(Exponent=ScalarParameter() {defaultValue=2, parameterName=FinalFresnelExponent}) {materialFunction=&MF_Fresnel_FX}.Fresnel
  )
  NamedRerouteDeclaration(None=Multiply(A=@n15, B=@n63)) {name=outEmissiveScore, nodeColor=<rgba 0.5938 0 1 1>}
  n64 = CurveAtlasRowParameter(CurveTime=NamedRerouteUsage()) {
   atlas=&CA_JumboTron
   curve=/Game/Accessories/FORT_Backpacks/Backpack_M_MED_Jumbotron_S20/Textures/C_JumboTron_Streaks
   defaultValue=2
   parameterName=Streaks
   sortPriority=0
  }
  n65 = MaterialFunctionCall(UVs=NamedRerouteUsage(), Texture Scale=@n64.R) {materialFunction=&ScaleUVsByCenter}
  n66 = TextureSample(UVs=Rotator(Coordinate=@n65.UVs) {speed=0.944}) {
   samplerSource=SSM_Clamp_WorldGroupSettings
   samplerType=SAMPLERTYPE_LinearColor
   texture=/BRCosmetics/Accessories/FORT_Backpacks/Backpack_M_MED_Jumbotron_S20/Textures/T_Jumbotron_Symbols
  }
  n67 = TextureSample(UVs=Rotator(Coordinate=@n65.UVs) {speed=-0.52}) {
   samplerSource=SSM_Clamp_WorldGroupSettings
   samplerType=SAMPLERTYPE_LinearColor
   texture=&T_Jumbotron_Symbols
  }
  n68 = MaterialFunctionCall(UVs=NamedRerouteUsage(), Texture Scale=Constant() {r=1.1}) {materialFunction=&ScaleUVsByCenter}
  n69 = TextureSample(UVs=Rotator(Coordinate=@n68.UVs) {speed=-0.138}) {
   samplerSource=SSM_Clamp_WorldGroupSettings
   samplerType=SAMPLERTYPE_LinearColor
   texture=&T_Jumbotron_Symbols
  }
  n70 = TextureSample(UVs=Rotator(Coordinate=@n68.UVs) {speed=0.1239}) {
   samplerSource=SSM_Clamp_WorldGroupSettings
   samplerType=SAMPLERTYPE_LinearColor
   texture=&T_Jumbotron_Symbols
  }
  n71 = Max(A=@n69.R, B=@n70.R)
  NamedRerouteDeclaration(None=Max(A=Max(A=@n66.R, B=@n67.R), B=Multiply(A=@n71) {constB=0.8})) {name=CelebrationStreaks, nodeColor=<rgba 1 0 0.375 1>}
  NamedRerouteDeclaration(None=Multiply(A=@n11, B=@n63)) {name=outBacckgroundOnly, nodeColor=<rgba 1 0.8906 0 1>}
  n72 = NamedRerouteUsage()
  n73 = Multiply(A=@n72, B=@n25) {constB=5}
  n74 = SmoothStep(Value=ComponentMask(None=Frac(None=@n73)) {g=true}) {constMax=0.5, constMin=0.5}
  n75 = VectorParameter() {parameterName=CRT_Scanline_Color, defaultValue=<rgba 0 0.0885 0.003 1>}
  NamedRerouteDeclaration(None=Multiply(A=Multiply(A=@n74, B=@n75.RGB), B=ScalarParameter() {defaultValue=1.5, parameterName=CRT_ColorScale})) {name=CRTPattern, nodeColor=<rgba 1 0 0.3047 1>}
  n76 = TextureCoordinate() {coordinateIndex=1}
  n77 = ComponentMask(None=VectorParameter() {parameterName=UVs_Offset, defaultValue=<rgba -0.01 0.03 0 1>}.RGB) {g=true, r=true}
  n78 = Add(A=TextureCoordinate() {coordinateIndex=1, vTiling=50}, B=@n77)
  n79 = Round(None=MaterialFunctionCall(Divisor=Constant() {r=0.333}, Mod=Constant() {r=1000}) {materialFunction=&MF_ModuloTime}.Unclamped)
  n80 = AppendVector(
   A=ComponentMask(None=MaterialFunctionCall!(PrimaryTextureCoordinates=@n78, 2ndOrderTextureCoordinates, RandomSeed=@n79, NoiseDensityBias, Bypass Noise Influence Crosstalk) {materialFunction=/Game/Effects/Fort_Effects/Materials/_Functions/MF_RandomUVSample}.NoiseWithRange) {g=true}
   B=Constant()
  )
  n81 = Time()
  n82 = Sine(None=@n81) {period=3.6}
  n83 = Sine(None=@n81) {period=2.44}
  n84 = Round(None=Power(Base=Max(A=@n82, B=@n83)) {constExponent=20})
  n85 = Multiply(A=Add(A=Multiply(A=@n80, B=@n84)) {constB=-0.5}, B=ScalarParameter() {defaultValue=0.06, parameterName=DistortionScale})
  n86 = NamedRerouteDeclaration(None=Add(A=Add(A=@n76, B=@n77), B=Saturate(None=@n85))) {name=UVs, nodeColor=<rgba 1 0 0.5391 1>}
  n87 = Constant() {r=0.4}
  n88 = CurveAtlasRowParameter(CurveTime=NamedRerouteUsage()) {
   atlas=&CA_JumboTron
   curve=/Game/Accessories/FORT_Backpacks/Backpack_M_MED_Jumbotron_S20/Textures/C_JumboTron_Numbers
   defaultValue=3
   parameterName=Numbers
   sortPriority=0
  }
  n89 = MaterialFunctionCall(UVs=NamedRerouteDeclaration(None=Rotator(Coordinate=@n86, Time=@n87)) {name=RotatedUVs, nodeColor=<rgba 0 0.0156 1 1>}, Texture Scale=@n88.R) {materialFunction=&ScaleUVsByCenter}
  n90 = ScalarParameter() {defaultValue=4.0696, parameterName=U_Scale}
  n91 = ScalarParameter() {defaultValue=0.22, parameterName=V_Scale}
  n92 = NamedRerouteDeclaration(None=AppendVector(
    A=NamedRerouteDeclaration(None=@n90) {name=U_Scale, nodeColor=<rgba 0.3359 0 1 1>}
    B=NamedRerouteDeclaration(None=@n91) {name=V_Scale, nodeColor=<rgba 0 1 0 1>}
   )) {name=UVScale, nodeColor=<rgba 0 0.2266 1 1>}
  n93 = Add(
   A=NamedRerouteDeclaration(None=Multiply(A=@n89.UVs, B=@n92)) {name=Rotated+ScaledUVs, nodeColor=<rgba 0 0.8828 1 1>}
   B=AppendVector(A=ScalarParameter() {defaultValue=-0.5087, parameterName=U_Bias}, B=ScalarParameter() {defaultValue=-0.057, parameterName=V_Bias})
  )
  n94 = Constant()
  n95 = Divide(A=Constant() {r=10}) {constB=11}
  n96 = MaterialFunctionCall(Float2=@n93) {materialFunction=&BreakOutFloat2Components}
  n97 = Constant() {r=0.5}
  n98 = SphereMask(A=@n96.R, B=@n97, Radius=@n97, Hardness=Constant() {r=1})
  n99 = Divide(A=ScalarParameter() {parameterName=Char0}) {constB=11}
  n100 = SphereMask(A=@n96.R, B=Constant() {r=1.5}, Radius=Constant() {r=0.5}, Hardness=Constant() {r=1})
  n101 = ScalarParameter() {parameterName=Char1}
  n102 = Constant() {r=2.5}
  n103 = Constant() {r=0.5}
  n104 = Constant() {r=1}
  n105 = Add(
   A=Add(A=Multiply(A=@n95, B=@n98), B=Multiply(A=@n99, B=@n100))
   B=Multiply(A=Divide(A=@n101) {constB=11}, B=SphereMask(A=@n96.R, B=@n102, Radius=@n103, Hardness=@n104))
  )
  n106 = NamedRerouteUsage()
  n107 = TextureSample!(UVs=Add(A=@n93, B=AppendVector(A=@n94, B=@n105)), Tex, "DDX(UVs)"=DDX(Value=@n106), "DDY(UVs)"=DDY(Value=@n106), Apply View MipBias) {
   mipValueMode=TMVM_Derivative
   samplerSource=SSM_Wrap_WorldGroupSettings
   samplerType=SAMPLERTYPE_LinearColor
   texture=/BRCosmetics/Accessories/FORT_Backpacks/Backpack_M_MED_Jumbotron_S20/Textures/T_Jumbotron_Numbers
  }
  NamedRerouteDeclaration(None=@n107.R) {name=RedChannel CRT, nodeColor=<rgba 0.3828 1 0 1>}
  NamedRerouteDeclaration(None=@n107.G) {name=GreenChannel CRT, nodeColor=<rgba 0 1 0 1>}
  n108 = CurveAtlasRowParameter(CurveTime=NamedRerouteUsage()) {
   atlas=&CA_JumboTron
   curve=/Game/Accessories/FORT_Backpacks/Backpack_M_MED_Jumbotron_S20/Textures/C_JumboTron_BurstBG
   parameterName=Burst
   sortPriority=0
  }
  n109 = TextureSample!(
   UVs=MaterialFunctionCall(UVs=NamedRerouteUsage(), Texture Scale=@n108.R) {materialFunction=&ScaleUVsByCenter}.UVs
   Tex
   Apply View MipBias
  ) {
   samplerSource=SSM_Clamp_WorldGroupSettings
   samplerType=SAMPLERTYPE_LinearColor
   texture=&T_Jumbotron_Symbols
  }
  NamedRerouteDeclaration(None=@n109.G) {name=CelebrationBurst, nodeColor=<rgba 1 0.7266 0 1>}
  n110 = TextureSampleParameter2D() {
   group=Jumbotron
   parameterName=Normal Map
   samplerType=SAMPLERTYPE_Normal
   texture=/BRCosmetics/Accessories/FORT_Backpacks/Backpack_M_MED_FNCSShield_S24/Textures/T_M_MED__FNCSShield_S24_Backpack_N
  }
  NamedRerouteDeclaration(None=@n110.RGB) {name=outNormalMap, nodeColor=<rgba 0.4766 1 0 1>}
  n111 = CurveAtlasRowParameter(CurveTime=NamedRerouteUsage()) {
   atlas=&CA_JumboTron
   curve=/Game/Accessories/FORT_Backpacks/Backpack_M_MED_Jumbotron_S20/Textures/C_JumboTron_Ribbons
   defaultValue=1
   parameterName=Ribbons
   sortPriority=0
  }
  n112 = TextureSample(UVs=MaterialFunctionCall(UVs=NamedRerouteUsage(), Texture Scale=@n111.R) {materialFunction=&ScaleUVsByCenter}.UVs) {
   samplerSource=SSM_Clamp_WorldGroupSettings
   samplerType=SAMPLERTYPE_LinearColor
   texture=&T_Jumbotron_Symbols
  }
  NamedRerouteDeclaration(None=@n112.B) {name=Ribbons, nodeColor=<rgba 0.4297 0 1 1>}
  n113 = ScalarParameter() {parameterName=Timestamp}
  n114 = ScalarParameter() {defaultValue=3, parameterName=TransitionLength}
  NamedRerouteDeclaration(None=VertexInterpolator(VS=MaterialFunctionCall(TimestampA=@n113, TransitionLength_In=@n114) {materialFunction=&MF_Timestamp_InOut_01}.Linear).PS) {name=TimeStamp, nodeColor=<rgba 0.125 1 0 1>}
  NamedRerouteUsage()
  Subtract(A=TextureCoordinate()) {constB=0.5}
  n115 = Multiply(A=@n72, B=@n25)
  n116 = Sine(None=ComponentMask(None=@n115) {g=true}) {period=2}
  n117 = AppendVector(A=Sign(None=NamedRerouteDeclaration(None=@n116) {name=CRT_UVs_Sine, nodeColor=<rgba 0.8984 1 0 1>}), B=Constant())
  TextureSample(UVs=Add(A=@n115, B=Multiply(A=@n117) {constB=0.25})) {samplerType=SAMPLERTYPE_LinearColor, texture=&T_Jumbotron_CRT}
 ]
}
