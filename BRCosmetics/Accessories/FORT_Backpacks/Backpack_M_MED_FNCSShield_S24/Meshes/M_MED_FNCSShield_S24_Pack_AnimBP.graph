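# Anim Blueprint for the FNCS Shield (S24) back bling that makes its four wing bones float gently back and forth.
# Each update, the event graph turns real time into two sine offsets along X at slightly different speeds, one for upper wings and one for lower.
# Chained Transform Bone nodes add these as bone-space translations before Output Pose.
# A spring-based AnimDynamics chain for the same bones exists but is not connected to the output.
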AnimBlueprint /BRCosmetics/Accessories/FORT_Backpacks/Backpack_M_MED_FNCSShield_S24/Meshes/M_MED_FNCSShield_S24_Pack_AnimBP
target_skeleton=&/M_MED_FNCSShield_S24_Pack_Skeleton
graphs [
 AnimGraph {nodes [
   n1 = Anim Dynamics - Bone: dyn_L_upper_wing(Alpha=1, ExternalForce="0,0,0") {
    angularDampingOverride=1
    angularSpringConstant=20000
    bLinearSpring=true
    bOverrideLinearDamping=true
    bUseGravityOverride=true
    gravityScale=0.35
    linearDampingOverride=1
    linearSpringConstant=200
    simulationSpace=RootRelative
    alphaBoolBlend {blendOption=Linear}
    alphaScaleBiasClamp {scale=0.8}
    boundBone {boneName=dyn_L_upper_wing}
    componentLinearAccScale=<0.5 0.5 0.5>
    componentLinearVelScale=<0.5 0.5 0.5>
    gravityOverride {z=-980}
    physicsBodyDefinitions [{
      boundBone {boneName=dyn_L_upper_wing}
      boxExtents=<300 300 300>
      constraintSetup {angularLimitsMax=<0 0 0>, angularLimitsMin=<0 0 0>, angularTarget=<1 0 0>, linearAxesMax=<2 0 4>, linearAxesMin=<0 -0.5 0>}
      localJointOffset=<1 0 0>
     }]
   }
   n2 = Anim Dynamics - Bone: dyn_L_lower_wing(ComponentPose=@n1, Alpha=1, ExternalForce="0,0,0") {
    angularDampingOverride=1
    angularSpringConstant=20000
    bLinearSpring=true
    bOverrideLinearDamping=true
    bUseGravityOverride=true
    gravityScale=0.35
    linearDampingOverride=1
    linearSpringConstant=100
    simulationSpace=RootRelative
    alphaBoolBlend {blendOption=Linear}
    alphaScaleBiasClamp {scale=0.8}
    boundBone {boneName=dyn_L_lower_wing}
    componentLinearAccScale=<0.46 0.46 0.46>
    componentLinearVelScale=<0.46 0.46 0.46>
    gravityOverride {z=-980}
    physicsBodyDefinitions [{
      boundBone {boneName=dyn_L_lower_wing}
      boxExtents=<300 300 300>
      constraintSetup {angularLimitsMax=<0 0 0>, angularLimitsMin=<0 0 0>, angularTarget=<1 0 0>, linearAxesMax=<0.5 0 4>, linearAxesMin=<0 -0.5 0>}
      localJointOffset=<1 0 0>
     }]
   }
   n3 = Anim Dynamics - Bone: dyn_R_upper_wing(ComponentPose=@n2, Alpha=1, ExternalForce="0,0,0") {
    angularDampingOverride=1
    angularSpringConstant=20000
    bLinearSpring=true
    bOverrideLinearDamping=true
    bUseGravityOverride=true
    gravityScale=0.35
    linearDampingOverride=1
    linearSpringConstant=200
    simulationSpace=RootRelative
    alphaBoolBlend {blendOption=Linear}
    alphaScaleBiasClamp {scale=0.8}
    boundBone {boneName=dyn_R_upper_wing}
    componentLinearAccScale=<0.5 0.5 0.5>
    componentLinearVelScale=<0.5 0.5 0.5>
    gravityOverride {z=-980}
    physicsBodyDefinitions [{
      boundBone {boneName=dyn_R_upper_wing}
      boxExtents=<300 300 300>
      constraintSetup {angularLimitsMax=<0 0 0>, angularLimitsMin=<0 0 0>, angularTarget=<1 0 0>, linearAxesMax=<2 0 4>, linearAxesMin=<0 -0.5 0>}
      localJointOffset=<1 0 0>
     }]
   }
   Anim Dynamics - Bone: dyn_R_lower_wing(ComponentPose=@n3, Alpha=1, ExternalForce="0,0,0") {
    angularDampingOverride=1
    angularSpringConstant=20000
    bLinearSpring=true
    bOverrideLinearDamping=true
    bUseGravityOverride=true
    gravityScale=0.35
    linearDampingOverride=1
    linearSpringConstant=100
    simulationSpace=RootRelative
    alphaBoolBlend {blendOption=Linear}
    alphaScaleBiasClamp {scale=0.8}
    boundBone {boneName=dyn_R_lower_wing}
    componentLinearAccScale=<0.46 0.46 0.46>
    componentLinearVelScale=<0.46 0.46 0.46>
    gravityOverride {z=-980}
    physicsBodyDefinitions [{
      boundBone {boneName=dyn_R_lower_wing}
      boxExtents=<300 300 300>
      constraintSetup {angularLimitsMax=<0 0 0>, angularLimitsMin=<0 0 0>, angularTarget=<1 0 0>, linearAxesMax=<0.5 0 4>, linearAxesMin=<0 -0.5 0>}
      localJointOffset=<1 0 0>
     }]
   }
   n4 = "Get FloaterA (More)"() -> ("FloaterA (More)":Vector)
   n5 = "Transform (Modify) Bone - Bone: dyn_L_upper_wing"(ComponentPose=Local To Component(LocalPose=Input Pose()), Alpha=1, Translation=@n4, Rotation="0,0,0", Scale="1,1,1") {translationMode=BMM_Additive, translationSpace=BCS_BoneSpace, alphaBoolBlend {blendOption=Linear}, boneToModify {boneName=dyn_L_upper_wing}}
   n6 = "Get FloaterB (Less)"() -> ("FloaterB (Less)":Vector)
   n7 = "Transform (Modify) Bone - Bone: dyn_L_lower_wing"(ComponentPose=@n5, Alpha=1, Translation=@n6, Rotation="0,0,0", Scale="1,1,1") {translationMode=BMM_Additive, translationSpace=BCS_BoneSpace, alphaBoolBlend {blendOption=Linear}, boneToModify {boneName=dyn_L_lower_wing}}
   n8 = "Transform (Modify) Bone - Bone: dyn_R_upper_wing"(ComponentPose=@n7, Alpha=1, Translation=@n4, Rotation="0,0,0", Scale="1,1,1") {translationMode=BMM_Additive, translationSpace=BCS_BoneSpace, alphaBoolBlend {blendOption=Linear}, boneToModify {boneName=dyn_R_upper_wing}}
   n9 = "Transform (Modify) Bone - Bone: dyn_R_lower_wing"(ComponentPose=@n8, Alpha=1, Translation=@n6, Rotation="0,0,0", Scale="1,1,1") {translationMode=BMM_Additive, translationSpace=BCS_BoneSpace, alphaBoolBlend {blendOption=Linear}, boneToModify {boneName=dyn_R_lower_wing}}
   Output Pose(Result=Component To Local(ComponentPose=@n9)) {layerGroup=None}
  ]}
 EventGraph {nodes [
   Event BlueprintUpdateAnimation()
   n1 = "Sin (Radians)"(A=float * float(A=GetRealTimeSeconds(), B=2))
   > "Set FloaterA (More)"("FloaterA (More)":Vector=MakeVector(X=float * float(A=@n1, B=1.2), Y=0)) -> (Output_Get:Vector)
   n2 = "Sin (Radians)"(A=float * float(A=GetRealTimeSeconds(), B=1.75))
   > "Set FloaterB (Less)"("FloaterB (Less)":Vector=MakeVector(X=float * float(A=@n2, B=1.2), Y=0)) -> (Output_Get:Vector)
   TryGetPawnOwner()
  ]}
]
variables [
 "FloaterA (More)" {type=Vector}
 "FloaterB (Less)" {type=Vector}
]
